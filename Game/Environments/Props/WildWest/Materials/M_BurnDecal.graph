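# Deferred decal material for a burn mark, with glowing embers that erode over the decal's lifetime.
# Perlin noise can distort the burn texture's UVs, while panning cloud noise varies ember color and intensity.
# An expanding, drifting sphere mask driven by DecalLifetimeOpacity eats away a soft radial glow, which is boosted to HDR.
# A slope mask limits opacity, with optional lifetime fade, and mobile zeroes base color and opacity.
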Material /Game/Environments/Props/WildWest/Materials/M_BurnDecal
graph {
 properties {blendMode=BLEND_Modulate, materialDomain=MD_DeferredDecal, pixelDepthOffsetMode=PDOM_Legacy, usedShadingModels=Default Lit}
 nodes [
  Constant2Vector() {g=1.6, r=1.1}
  Constant3Vector()
  Constant() {r=0.1}
  Constant() {r=2}
  MaterialFunctionCall() {materialFunction=&DebugTimeSine}
  n1 = VectorParameter() {parameterName=ColorLow, defaultValue=<rgba 1 0.3294 0.1183 1>}
  n2 = VectorParameter() {parameterName=ColorHigh, defaultValue=<rgba 1 0.6484 0.217 1>}
  n3 = TextureSample(UVs=Panner(Coordinate=TextureCoordinate() {uTiling=1.2, vTiling=1.5}) {speedX=0.05, speedY=0.05}) {samplerType=SAMPLERTYPE_LinearColor, texture=/Game/VisualThreatManager/StormVisuals/Textures/Clouds_VectorDisplacement}
  n4 = ComponentMask(None=TextureSample() {texture=/Game/Global/Textures/Default/Noise/3xPerlin_M}.RGB) {g=true, r=true}
  n5 = LinearInterpolate(Alpha=ScalarParameter() {parameterName=UV_Distortion}) {constB=0.3}
  n6 = TextureSampleParameter2D(UVs=Add(A=TextureCoordinate(), B=Multiply(A=@n4, B=@n5))) {parameterName=BaseColorTexture, samplerType=SAMPLERTYPE_Masks, texture=/Game/Environments/Props/WildWest/Textures/T_BurnDecal}
  n7 = OneMinus(None=LinearInterpolate(A=@n6.R, B=@n6.R, Alpha=ScalarParameter() {parameterName=UseTextureAlpha}))
  n8 = MaterialFunctionCall(
   UVs=TextureCoordinate()
   Texture Scale=AppendVector(
    A=ScalarParameter() {defaultValue=1.1, parameterName=Emissive_U_Scale}
    B=ScalarParameter() {defaultValue=1.6, parameterName=Emissive_V_Scale}
   )
  ) {materialFunction=&ScaleUVsByCenter}
  n9 = Multiply(A=DecalLifetimeOpacity().Opacity, B=ScalarParameter() {defaultValue=1, parameterName=EmissiveErode})
  n10 = OneMinus(None=@n9)
  n11 = AppendVector(A=LinearInterpolate(Alpha=@n10) {constA=0.45, constB=0.2}, B=LinearInterpolate(Alpha=@n10) {constA=0.4, constB=0.5})
  n12 = LinearInterpolate(Alpha=@n10) {constA=0.2}
  n13 = LinearInterpolate(Alpha=@n3.G) {constB=0.4}
  n14 = Power(Base=@n9) {constExponent=5}
  n15 = Subtract(
   A=SphereMask(A=TextureCoordinate(), B=Constant() {r=0.5}, Radius=LinearInterpolate(Alpha=@n3.R) {constA=0.45, constB=0.55}) {attenuationRadius=0.5, hardnessPercent=0}
   B=LinearInterpolate(A=SphereMask(A=@n8.UVs, B=@n11, Radius=@n12, Hardness=@n13) {attenuationRadius=0.4, hardnessPercent=40}, Alpha=@n14) {constB=0}
  )
  n16 = Multiply(A=Multiply(A=LinearInterpolate(A=@n1.RGB, B=@n2.RGB, Alpha=@n3.G), B=@n3.R), B=Multiply(A=@n7, B=Saturate(None=@n15)))
  n17 = Multiply(A=@n16) {constB=5}
  Multiply(
   A=ScalarParameter() {parameterName=OverallEmissive}
   B=MaterialFunctionCall(LDR Input=Saturate(None=@n17), HDR Tint=Multiply(A=@n16) {constB=250}) {materialFunction=&DeriveHDRfromLDR}.RemappedRange
  )
  ScalarParameter() {parameterName=Metallic}
  ShadingPathSwitch(Default=@n6.RGB, Mobile=Constant()) {desc=Migrated from FeatureLevelSwitch}
  n18 = ObjectOrientation()
  n19 = Multiply(
   A=Multiply(A=@n7, B=ScalarParameter() {defaultValue=3.9594, parameterName=OverallOpacity})
   B=OneMinus(None=MaterialFunctionCall!(TangentNormal, SlopAngle=@n18, FalloffPower, CheapContrast) {materialFunction=/Engine/Functions/Engine_MaterialFunctions01/Texturing/SlopeMask}.Result)
  )
  ShadingPathSwitch(Default=LinearInterpolate(A=@n19, B=Multiply(A=@n14, B=@n19), Alpha=ScalarParameter() {parameterName=FadeBaseColor}), Mobile=Constant()) {desc=Migrated from FeatureLevelSwitch}
  TextureCoordinate()
 ]
}
